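# Anim Blueprint for the claw machine back bling, animating a crane claw in reaction to the wearer's kills.
# Cosmetic stat watchers flag kill increases, starting a delay-chained sequence of alphas that interpolated bone transforms use to pose fingers and shift the crane.
# Bear bones get faint spring motion, disabled at high pawn speed, while AnimDynamics springs the claw belt.
# Damage increases update variables the graphs never read.
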AnimBlueprint /BRCosmetics/Accessories/FORT_Backpacks/Backpack_M_MED_ClawMachine/Meshes/M_MED_ClawMachine_Pack_AnimBP
target_skeleton=&/M_MED_ClawMachine_Pack_Skeleton
graphs [
 AnimGraph {nodes [
   n1 = Input Pose()
   n2 = Get SpringAlpha() -> (SpringAlpha:double)
   n3 = Spring controller - Bone: Bear_02_JNT!(
    ComponentPose:Component Space Pose Link Structure=Spring controller - Bone: Bear_01_JNT!(ComponentPose:Component Space Pose Link Structure=Local To Component(LocalPose=@n1), Alpha:float=@n2) -> (Pose:Component Space Pose Link Structure) {
     is=title AnimGraphNode_SpringBone
     springDamping=1.5
     springStiffness=150
     alphaBoolBlend {blendOption=Linear}
     springBone {boneName=Bear_01_JNT}
    }
    Alpha:float=@n2
   ) -> (Pose:Component Space Pose Link Structure) {
    is=title AnimGraphNode_SpringBone
    springDamping=2
    springStiffness=150
    alphaBoolBlend {blendOption=Linear}
    springBone {boneName=Bear_02_JNT}
   }
   n4 = Get Spring Alpha 02() -> (Spring Alpha 02:double)
   n5 = Get Crane Alpha() -> (Crane Alpha:double)
   n6 = "Transform (Modify) Bone - Bone: Claw_01_Finger_01_JNT"!(
    ComponentPose=Spring controller - Bone: Bear_04_JNT!(
     ComponentPose:Component Space Pose Link Structure=Spring controller - Bone: Bear_03_JNT!(ComponentPose:Component Space Pose Link Structure=@n3, Alpha:float=@n4) -> (Pose:Component Space Pose Link Structure) {
      is=title AnimGraphNode_SpringBone
      springDamping=2
      springStiffness=200
      alphaBoolBlend {blendOption=Linear}
      springBone {boneName=Bear_03_JNT}
     }
     Alpha:float=@n4
    ) -> (Pose:Component Space Pose Link Structure) {
     is=title AnimGraphNode_SpringBone
     springDamping=1.5
     springStiffness=200
     alphaBoolBlend {blendOption=Linear}
     springBone {boneName=Bear_04_JNT}
    }
    Alpha=@n5
    Translation="0,0,0"
    Rotation="30,0,0"
   ) -> (Pose) {
    alpha=0
    rotationMode=BMM_Replace
    rotationSpace=BCS_BoneSpace
    scaleSpace=BCS_BoneSpace
    translationMode=BMM_Additive
    translationSpace=BCS_BoneSpace
    alphaBoolBlend {blendOption=Linear}
    alphaScaleBiasClamp {bInterpResult=true, interpSpeedDecreasing=5, interpSpeedIncreasing=5}
    boneToModify {boneName=Claw_01_Finger_01_JNT}
    rotation {pitch=30}
   }
   n7 = "Transform (Modify) Bone - Bone: Claw_02_Finger_01_JNT"!(ComponentPose=@n6, Alpha=@n5, Translation="0,0,0", Rotation="30,0,0") -> (Pose) {
    alpha=0
    rotationMode=BMM_Replace
    rotationSpace=BCS_BoneSpace
    scaleSpace=BCS_BoneSpace
    translationMode=BMM_Additive
    translationSpace=BCS_BoneSpace
    alphaBoolBlend {blendOption=Linear}
    alphaScaleBiasClamp {bInterpResult=true, interpSpeedDecreasing=5, interpSpeedIncreasing=5}
    boneToModify {boneName=Claw_02_Finger_01_JNT}
    rotation {pitch=30}
   }
   n8 = "Transform (Modify) Bone - Bone: Claw_03_Finger_01_JNT"!(ComponentPose=@n7, Alpha=@n5, Translation="0,0,0", Rotation="30,0,0") -> (Pose) {
    alpha=0
    rotationMode=BMM_Replace
    rotationSpace=BCS_BoneSpace
    scaleSpace=BCS_BoneSpace
    translationMode=BMM_Additive
    translationSpace=BCS_BoneSpace
    alphaBoolBlend {blendOption=Linear}
    alphaScaleBiasClamp {bInterpResult=true, interpSpeedDecreasing=5, interpSpeedIncreasing=5}
    boneToModify {boneName=Claw_03_Finger_01_JNT}
    rotation {pitch=30}
   }
   n9 = "Transform (Modify) Bone - Bone: Claw_01_Finger_02_JNT"!(ComponentPose=@n8, Alpha=@n5, Translation="0,0,0", Rotation="30,0,0") -> (Pose) {
    alpha=0
    rotationMode=BMM_Replace
    rotationSpace=BCS_BoneSpace
    scaleSpace=BCS_BoneSpace
    translationMode=BMM_Additive
    translationSpace=BCS_BoneSpace
    alphaBoolBlend {blendOption=Linear}
    alphaScaleBiasClamp {bInterpResult=true, interpSpeedDecreasing=5, interpSpeedIncreasing=5}
    boneToModify {boneName=Claw_01_Finger_02_JNT}
    rotation {pitch=30}
   }
   n10 = "Transform (Modify) Bone - Bone: Claw_02_Finger_02_JNT"!(ComponentPose=@n9, Alpha=@n5, Translation="0,0,0", Rotation="30,0,0") -> (Pose) {
    alpha=0
    rotationMode=BMM_Replace
    rotationSpace=BCS_BoneSpace
    scaleSpace=BCS_BoneSpace
    translationMode=BMM_Additive
    translationSpace=BCS_BoneSpace
    alphaBoolBlend {blendOption=Linear}
    alphaScaleBiasClamp {bInterpResult=true, interpSpeedDecreasing=5, interpSpeedIncreasing=5}
    boneToModify {boneName=Claw_02_Finger_02_JNT}
    rotation {pitch=30}
   }
   n11 = "Transform (Modify) Bone - Bone: Claw_03_Finger_02_JNT"!(ComponentPose=@n10, Alpha=@n5, Translation="0,0,0", Rotation="30,0,0") -> (Pose) {
    alpha=0
    rotationMode=BMM_Replace
    rotationSpace=BCS_BoneSpace
    scaleSpace=BCS_BoneSpace
    translationMode=BMM_Additive
    translationSpace=BCS_BoneSpace
    alphaBoolBlend {blendOption=Linear}
    alphaScaleBiasClamp {bInterpResult=true, interpSpeedDecreasing=5, interpSpeedIncreasing=5}
    boneToModify {boneName=Claw_03_Finger_02_JNT}
    rotation {pitch=30}
   }
   n12 = Anim Dynamics - Bone: Claw_Belt_JNT(ComponentPose=@n11, Alpha=1, ExternalForce="0,0,0") {
    angularDampingOverride=0.9
    angularSpringConstant=200000
    bAngularSpring=true
    bLinearSpring=true
    bOverrideAngularDamping=true
    bOverrideLinearDamping=true
    gravityScale=0
    linearDampingOverride=0.995
    linearSpringConstant=50
    simulationSpace=BoneRelative
    alphaBoolBlend {blendOption=Linear}
    boundBone {boneName=Claw_Belt_JNT}
    componentAppliedLinearAccClamp=<6000 6000 6000>
    componentLinearAccScale=<0.3 0.3 0.3>
    componentLinearVelScale=<0.7 0.7 0.7>
    gravityOverride {z=-980}
    physicsBodyDefinitions [{
      sphereCollisionRadius=0
      boundBone {boneName=Claw_Belt_JNT}
      boxExtents=<7 7 7>
      constraintSetup {
       angularTargetAxis=AxisZ
       twistAxis=AxisZ
       angularLimitsMax=<0 0 0>
       angularLimitsMin=<0 0 0>
       angularTarget=<0 0 0>
       linearAxesMax=<0.05 0.05 0.05>
       linearAxesMin=<-0.02 -0.02 -10>
      }
      localJointOffset=<-0.1 -0.2 0>
     }]
    relativeSpaceBone {boneName=pack}
   }
   n13 = "Transform (Modify) Bone - Bone: Claw_Crane_JNT"!(ComponentPose=@n12, Alpha=Get Crane Alpha 02() -> (Crane Alpha 02:double), Translation="3.5,0,0", Rotation="0,0,0") -> (Pose) {
    alpha=0
    rotationMode=BMM_Additive
    rotationSpace=BCS_BoneSpace
    scaleSpace=BCS_BoneSpace
    translationMode=BMM_Additive
    translationSpace=BCS_BoneSpace
    alphaBoolBlend {blendOption=Linear}
    alphaScaleBiasClamp {bInterpResult=true, interpSpeedDecreasing=5, interpSpeedIncreasing=5}
    boneToModify {boneName=Claw_Crane_JNT}
    translation {x=3.5}
   }
   n14 = Get Crane Alpha 03() -> (Crane Alpha 03:double)
   n15 = "Transform (Modify) Bone - Bone: Claw_01_Finger_01_JNT"(ComponentPose=@n13, Alpha=@n14, Translation="0,0,0", Rotation="-15,0,0") {
    alpha=0
    rotationMode=BMM_Replace
    rotationSpace=BCS_BoneSpace
    scaleSpace=BCS_BoneSpace
    translationMode=BMM_Additive
    translationSpace=BCS_BoneSpace
    alphaBoolBlend {blendOption=Linear}
    alphaScaleBiasClamp {bInterpResult=true, interpSpeedDecreasing=5, interpSpeedIncreasing=5}
    boneToModify {boneName=Claw_01_Finger_01_JNT}
    rotation {pitch=-15}
   }
   n16 = "Transform (Modify) Bone - Bone: Claw_02_Finger_01_JNT"(ComponentPose=@n15, Alpha=@n14, Translation="0,0,0", Rotation="-15,0,0") {
    alpha=0
    rotationMode=BMM_Replace
    rotationSpace=BCS_BoneSpace
    scaleSpace=BCS_BoneSpace
    translationMode=BMM_Additive
    translationSpace=BCS_BoneSpace
    alphaBoolBlend {blendOption=Linear}
    alphaScaleBiasClamp {bInterpResult=true, interpSpeedDecreasing=5, interpSpeedIncreasing=5}
    boneToModify {boneName=Claw_02_Finger_01_JNT}
    rotation {pitch=-15}
   }
   n17 = "Transform (Modify) Bone - Bone: Claw_03_Finger_01_JNT"(ComponentPose=@n16, Alpha=@n14, Translation="0,0,0", Rotation="-15,0,0") {
    alpha=0
    rotationMode=BMM_Replace
    rotationSpace=BCS_BoneSpace
    scaleSpace=BCS_BoneSpace
    translationMode=BMM_Additive
    translationSpace=BCS_BoneSpace
    alphaBoolBlend {blendOption=Linear}
    alphaScaleBiasClamp {bInterpResult=true, interpSpeedDecreasing=5, interpSpeedIncreasing=5}
    boneToModify {boneName=Claw_03_Finger_01_JNT}
    rotation {pitch=-15}
   }
   n18 = "Transform (Modify) Bone - Bone: Claw_01_Finger_02_JNT"(ComponentPose=@n17, Alpha=@n14, Translation="0,0,0", Rotation="-15,0,0") {
    alpha=0
    rotationMode=BMM_Replace
    rotationSpace=BCS_BoneSpace
    scaleSpace=BCS_BoneSpace
    translationMode=BMM_Additive
    translationSpace=BCS_BoneSpace
    alphaBoolBlend {blendOption=Linear}
    alphaScaleBiasClamp {bInterpResult=true, interpSpeedDecreasing=5, interpSpeedIncreasing=5}
    boneToModify {boneName=Claw_01_Finger_02_JNT}
    rotation {pitch=-15}
   }
   n19 = "Transform (Modify) Bone - Bone: Claw_02_Finger_02_JNT"(ComponentPose=@n18, Alpha=@n14, Translation="0,0,0", Rotation="-15,0,0") {
    alpha=0
    rotationMode=BMM_Replace
    rotationSpace=BCS_BoneSpace
    scaleSpace=BCS_BoneSpace
    translationMode=BMM_Additive
    translationSpace=BCS_BoneSpace
    alphaBoolBlend {blendOption=Linear}
    alphaScaleBiasClamp {bInterpResult=true, interpSpeedDecreasing=5, interpSpeedIncreasing=5}
    boneToModify {boneName=Claw_02_Finger_02_JNT}
    rotation {pitch=-15}
   }
   n20 = "Transform (Modify) Bone - Bone: Claw_03_Finger_02_JNT"(ComponentPose=@n19, Alpha=@n14, Translation="0,0,0", Rotation="-15,0,0") {
    alpha=0
    rotationMode=BMM_Replace
    rotationSpace=BCS_BoneSpace
    scaleSpace=BCS_BoneSpace
    translationMode=BMM_Additive
    translationSpace=BCS_BoneSpace
    alphaBoolBlend {blendOption=Linear}
    alphaScaleBiasClamp {bInterpResult=true, interpSpeedDecreasing=5, interpSpeedIncreasing=5}
    boneToModify {boneName=Claw_03_Finger_02_JNT}
    rotation {pitch=-15}
   }
   Output Pose(Result=Component To Local(ComponentPose=@n20)) {layerGroup=None}
  ]}
 EventGraph {nodes [
   Event BlueprintUpdateAnimation()
   > n1 = Sequence!(execute) -> (then_0, then_1, then_2:Exec, then_3:Exec, then_4:Exec, then_5:Exec, then_6:Exec)
   Branch(execute=@n1.then_2, Condition=Get Active_Kill() -> (Active_Kill:bool))
   > Set Crane_Timer(Crane_Timer:double=float + float(A=Get Crane_Timer() -> (Crane_Timer:double), B=1) {is=title K2Node_CommutativeAssociativeBinaryOperator}) -> (Output_Get:double)
   > n2 = Sequence!(execute) -> (then_0, then_1, then_2)
   > Set Active_Kill(Active_Kill:bool=false) -> (Output_Get:bool)
   Delay(execute=@n2.then_2, Duration=0.5)
   > Set Crane Alpha(Crane Alpha:double=1) -> (Output_Get:double)
   > Delay(Duration=0.5)
   > Set Crane Alpha(Crane Alpha=1)
   > Set Crane Alpha 02(Crane Alpha 02:double=1) -> (Output_Get:double)
   > Delay(Duration=1)
   > Set Crane Alpha(Crane Alpha=0)
   > Set Crane Alpha 02(Crane Alpha 02=1)
   > Set Crane Alpha 03(Crane Alpha 03:double=1) -> (Output_Get:double)
   > Delay(Duration=1)
   > Set Crane Alpha 02(Crane Alpha 02=0)
   > Set Crane Alpha 03(Crane Alpha 03=1)
   > Delay(Duration=0.5)
   > Set Crane Alpha 03(Crane Alpha 03=0)
   n3 = Set pawnVelocitySpeed(execute=@n1.then_3, pawnVelocitySpeed:double=Vector Length(A=Get PawnVelocity())) -> (Output_Get:double)
   > n4 = Set pawnVelocitySpeedCutoff(pawnVelocitySpeedCutoff:double=SelectFloat(A=0, B=1, bPickA="float > float"(A=@n3, B=1200) {is=title K2Node_CallFunction})) -> (Output_Get:double)
   > n5 = Set SpringAlpha(SpringAlpha:double=float * float(A=0.01, B=@n4) {is=title K2Node_PromotableOperator}) -> (Output_Get:double)
   > Set Spring Alpha 02(Spring Alpha 02:double=float * float(A=0.5, B=@n5)) -> (Output_Get:double)
   Event OnCharacterCustomizationLoaded()
   > n6 = Sequence!(execute) -> (then_0, then_1)
   > n7 = Cast To FortPlayerPawn(Object=TryGetFortPawn())
   > n8 = WatchCosmeticStat!(
    execute:Exec
    Pawn:Fort Player Pawn Object Reference=@n7
    StatTag:Fort Stat Manager Tag Structure=(TagName="StatManager.AthenaKills")
    DefaultValue:int=0
    bBroadcastInitialValue:bool=true
    InitialValueBroadcastDelay:float=-1
    ChangeValueBroadcastDelay:float=-1
    bConsumeWaitingNotifications:bool=true
   ) -> (then:Exec, OnStatChanged:Exec, StatName:Name, StatValue:int) {is=title K2Node_AsyncAction}
   n9 = Set KillCount(execute=@n8.OnStatChanged, KillCount:double="To Float (Integer)"(InInt=@n8.StatValue)) -> (Output_Get:double)
   > Branch(Condition="float > float"(A=@n9, B=Get Last_KillCount() -> (Last_KillCount:double)))
   > Set Active_Kill(Active_Kill=true)
   > Set Last_KillCount(Last_KillCount:double=@n9) -> (Output_Get:double)
   n10 = Cast To FortPlayerPawn(execute=@n6.then_1, Object=TryGetFortPawn())
   > n11 = WatchCosmeticStat(
    Pawn=@n10
    StatTag=(TagName="StatManager.DamageDealt")
    DefaultValue=0
    bBroadcastInitialValue=true
    InitialValueBroadcastDelay=-1
    ChangeValueBroadcastDelay=-1
    bConsumeWaitingNotifications=true
   )
   n12 = Set DamageCount(execute=@n11.OnStatChanged, DamageCount:double="To Float (Integer)"(InInt=@n11.StatValue)) -> (Output_Get:double)
   > Branch(Condition="float > float"(A=@n12, B=Get Last Damage Count() -> (Last Damage Count:double)))
   > Set Active_Damage(Active_Damage:bool=true) -> (Output_Get:bool)
   > Set Last Damage Count(Last Damage Count:double=@n12) -> (Output_Get:double)
   > Set Claw Count(Claw Count:double=float + float(A=Get Claw Count() -> (Claw Count:double), B=-180)) -> (Output_Get:double)
   TryGetPawnOwner()
   MapRangeClamped(
    Value=float - float(A=Get Crane Timer Length() -> (Crane Timer Length:double), B=Get Crane_Timer()) {is=title K2Node_CallFunction}
    InRangeA=Get Crane Timer Length()
    InRangeB=0
    OutRangeA=1
    OutRangeB=0
   )
  ]}
]
variables [
 Active_Damage {type=boolean}
 Active_Kill {type=boolean}
 Claw Count {type=number}
 Crane Alpha {type=number}
 Crane Alpha 02 {type=number}
 Crane Alpha 03 {type=number}
 Crane Timer Length {type=number}
 Crane_Timer {type=number}
 DamageCount {type=number}
 KillCount {type=number}
 Last Damage Count {type=number}
 Last_KillCount {type=number}
 pawnVelocitySpeed {type=number}
 pawnVelocitySpeedCutoff {type=number}
 Spring Alpha 02 {type=number}
 SpringAlpha {type=number}
]
